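; Gameplay ability for the Battle Flopper consumable that marks the nearest enemy player to the user, unless marking is disabled by data.
; On activation it caches radius and duration; the server gathers enemy player pawns in range, while non-server instances play a local ping cue.
; After a short timer, the closest one receives a timed indicator and red stencil shared with the squad, then the ability ends.
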(event Ability|EventActivateAbility
  (bind _returnvalue (Ability|GetAvatarActorfromActorInfo))
  (Utilities|FlowControl|DoOnce false
    (:Completed
      (if (not (Class|AthenaFunctionLibrary|RowtoBool (Variables|Default|GetDisableMarkingPlayers)))
        (Variables|CachedValues|SetCurrentMarkingRadius (ScalableFloat|GetValueatLevel (Variables|Default|GetMarkingRadius)))
        (Variables|CachedValues|SetCurrentMarkingDuration (ScalableFloat|GetValueatLevel (Variables|Default|GetMarkingDuration)))
        (if (Networking|IsServer)
          (Utilities|Array|Clear (Variables|Default|GetWorkingList))
          ;[1 1 1] Get players in range
          (Utilities|Casting|CastToFortPawn _returnvalue
            (:then
              (bind _outfortpawns (Fortnite|GetFortPawnsInRange (Transformation|GetActorLocation _returnvalue) (Variables|CachedValues|GetCurrentMarkingRadius) true))
              ;[1 1 1] Only check for enemy players
              (for _array_element _outfortpawns
                (Utilities|Casting|CastToFortPlayerPawnAthena _array_element)
                (if (not (Team|OnSameTeam (Ability|GetAvatarActorfromActorInfo) _array_element))
                  (Utilities|Array|AddUnique (Variables|Default|GetWorkingList) _array_element)))
              (Utilities|Time|SetTimerbyEvent (delegate Custom|MarkEnemies_) 0.8))
            (:CastFailed
              (Ability|EndAbility)))
          (else
            ;[1 1 1] Blue "ping" VFX
            (GameplayCue|ExecuteGameplayCue%28Burst%29Local (Ability|GetAvatarActorfromActorInfo) (Variables|Default|GetCueOnMark) (Ability|GameplayCue|MakeGameplayCueParameters :NormalizedMagnitude 0.0 :RawMagnitude (Variables|CachedValues|GetCurrentMarkingRadius) :Location (Transformation|GetActorLocation (Ability|GetAvatarActorfromActorInfo)) :Normal "0,0,0" :SourceObject (Ability|GetAvatarActorfromActorInfo) :GameplayEffectLevel 1 :AbilityLevel 1 :bReplicateLocationWhenUsingMinimalRepProxy false))))))))

(event Custom|MarkEnemies_
  (bind _returnvalue (Ability|GetAvatarActorfromActorInfo))
  (bind _workinglist (Variables|Default|GetWorkingList))
  (bind _returnvalue_2 (Fortnite|GetClosestActorfromArray _returnvalue _workinglist))
  (bind _array (Utilities|Array|MakeArray _returnvalue_2))
  (bind _currentmarkingduration (Variables|CachedValues|GetCurrentMarkingDuration))
  ;[1 1 1] Add indicator and stencil to nearest enemy (Mark Enemies)
  (bind _asfort_player_controller_athena (Utilities|Casting|CastToFortPlayerControllerAthena (Fortnite|GetFortPlayerControllerfromActor (Ability|GetAvatarActorfromActorInfo)))
    (:then
      (IndicatedManagement|AddActorsToIndicatedList _asfort_player_controller_athena _array (IndicatedManagement|MakeIndicatedActorData :GroupIdentifier "BattleFlopper" :ParticleSystemData (IndicatedManagement|MakeIndicatedActorParticleSystemData "/Game/Athena/Items/Gameplay/Lotus/Scooter/FX/P_Lotus_Scooter_Flare.P_Lotus_Scooter_Flare" "BoneSocketActor" "0, 0,100.000000" "0, 0,40.000000" "WorldOffset") :EventTags (Tags|GetAbilityTags) :Duration _currentmarkingduration :StepTime 0.1 :IndicatorOffset "0.000000,0.000000,90.000000" :IndicatorDBNOOffset "0.000000,0.000000,30.000000" :ShareActorWith "SquadOnTeam" :Sound "/Game/Sounds/Fort_Human_Barks/SFX/Mang_Interrogate_Indicated_Cue.Mang_Interrogate_Indicated_Cue" :StateImageOverride "None" :StateIndexOverride -1 :bClampToScreen true :bHideStateImage false :bOnlyDisplayWhenStencilingIsDisabled false :bOnlyDisplayIfActorIsVisible false))
      (IndicatedManagement|AddActorsToStenciledList _asfort_player_controller_athena _array (IndicatedManagement|MakeStenciledActorData :GroupIdentifier "BattleFlopper" :Duration _currentmarkingduration :StepTime 0.1 :ShareActorWith "SquadOnTeam" :FriendlyStencilName "None" :EnemyStencilName "RedOverlayID" :bOnlyDisplayIfActorIsVisible false))
      (Ability|EndAbility))
    (:CastFailed
      (Ability|EndAbility))))

(event Ability|EventOnEndAbility
  (params (bWasCancelled :type Boolean)))
